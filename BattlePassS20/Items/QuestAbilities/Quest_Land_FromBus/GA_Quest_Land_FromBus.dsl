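; GA_Quest_Land_FromBus is a Battle Pass quest ability, judging by its name and tags, for quests about landing from the bus at tagged locations.
; When activated by a gameplay event, it calls the parent, then checks whether the owning player's pawn has any of the configured landing location tags.
; On a match, it sends a complex custom stat event with the landed-from-bus tag to the player's quest manager and ends the ability.
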(event Ability|EventActivateAbilityFromEvent
  (params (EventData :type Gameplay%20Event%20Data%20Structure%20%28by%20ref%29))
  (|CallParentFunction EventData)
  (if (GameplayTags|HasAnyMatchingGameplayTags (Variables|GAVars|Data|GetOwningPlayerPawn) (Variables|Default|GetLandingLocationTags))
    (Quest|SendComplexCustomStatEvent (Variables|GAVars|Data|GetOwningPlayerQuestManager) 0 (Variables|Default|GetLandedFromBusTag) "(GameplayTags=)" 1)
    (Ability|EndAbility)))
